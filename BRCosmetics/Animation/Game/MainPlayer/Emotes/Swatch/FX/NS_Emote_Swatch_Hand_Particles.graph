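# Niagara effect for the Swatch emote's hand particles, built from three sprite emitters: a heart, its glow and sparks.
# The heart bursts with curve-driven color and flickering brightness; a purple glow scales in, offset behind it.
# GPU sparks spawn at a decaying rate with User Color Array tints, radial and inherited velocity, curl noise and speed-based stretching.
# Time-of-day scalars adjust brightness or opacity, with a front-end modifier on the heart.
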NiagaraSystem /BRCosmetics/Animation/Game/MainPlayer/Emotes/Swatch/FX/NS_Emote_Swatch_Hand_Particles
dependencies {
 usedDataInterfaces [&NiagaraDataInterfaceCurve, &NiagaraDataInterfaceColorCurve]
 usedDynamicInputs []
 usedModules [&SystemState]
}
emitters {
 Final_Heart_Lit001 {
  emitterUpdateScript [
   &EmitterState(Life Cycle Mode=::NewEnumerator1, Loop Behavior=::NewEnumerator1, UseLoopDelay=false)
   &SpawnBurst_Instantaneous(Use Spawn Probability=false, UseLoopCountLimit=false)
  ]
  particleSpawnScript [
   &InitializeParticle(Lifetime=1.3, Sprite Size Mode=::NewEnumerator3, Uniform Sprite Size=30, UsePositionOffset=false, Write Material Random=true)
   &CameraOffset(Camera Offset Amount=10)
  ]
  particleUpdateScript [
   &ParticleState()
   &DynamicMaterialParameters(
    Index 0 Param 1=2
    Index 0 Param 4=&RandomRangeFloat(Minimum=0, Maximum=1)
    Module.Param0WriteEnabled=true
    Module.Param0WriteEnabled1=true
    Module.Param1WriteEnabled=true
    Module.Param1WriteEnabled1=true
    Module.Param2WriteEnabled=true
    Module.Param2WriteEnabled1=true
    Module.Param3WriteEnabled=true
    Module.Param3WriteEnabled1=true
    Write Parameter Index 1=true
   )
   &Color(Color=&MultiplyLinearColor(
     Linear Color=&ColorFromCurve(ColorCurve=DataInterface() {propertyValues {
        AlphaCurve {keys [<at 0 1 RCIM_Cubic>]}
        BlueCurve {keys [<at 0.2171 1>,<at 0.47 0.5647>,<at 0.8272 1 RCIM_Cubic>]}
        GreenCurve {keys [<at 0.2171 0.0742>,<at 0.47 0.0844>,<at 0.8272 0 RCIM_Cubic>]}
        RedCurve {keys [<at 0.2171 0.0704>,<at 0.47 0.7835>,<at 0.8272 0.2 RCIM_Cubic>]}
       }})
     Scale Factor=&MakeFloat4FromVectorAndFloat("Vector (RGB)"=&VectorFromFloat(Value=10), "Float (Alpha)"=1)
    ))
   &ScaleColor(
    Color Value To Scale=$StackContext.Color
    Scale Alpha=&FloatFromCurve(FloatCurve=DataInterface() {propertyValues {Curve {keys [{arriveTangent=-1.3442, interpMode=RCIM_Cubic, leaveTangent=-1.3442, tangentMode=RCTM_User, time=0, value=1}, <at 1 0 RCIM_Cubic>]}}})
    Scale RGB=&VectorFromFloat(Value=&Multiply_Float(
      A=&FloatFromCurve(FloatCurve=DataInterface() {propertyValues {Curve {keys [
           {arriveTangent=54.0067, interpMode=RCIM_Cubic, time=0.1, value=5}
           <at 0.2 1 RCIM_Cubic>
           {arriveTangent=-0.0353, interpMode=RCIM_Cubic, leaveTangent=-0.0353, tangentMode=RCTM_User, time=0.3, value=5}
           {arriveTangent=0.0286, interpMode=RCIM_Cubic, leaveTangent=0.0286, tangentMode=RCTM_User, time=0.65, value=5}
           <at 0.75 0.3 RCIM_Cubic>
           <at 0.8 5 RCIM_Cubic>
          ]}}})
      B=&Multiply_Float(
       A=&FNTimeOfDayScalar(Night Scale=0.7, Morning Scale=1, Evening Scale=1)
       B=&FN_FrontEndFloatModifier(Gameplay=1, Front End=1)
      )
     ))
    ScaleA=true
    ScaleRGB=true
   )
  ]
  rendererClasses [&NiagaraSpriteRendererProperties]
  renderers [{rendererClass=&NiagaraSpriteRendererProperties}]
 }
 Heart_Glow {
  emitterUpdateScript [
   &EmitterState(Life Cycle Mode=::NewEnumerator1, Loop Behavior=::NewEnumerator1, Loop Delay=0.05, Loop Duration=1, UseLoopDelay=true)
   &SpawnBurst_Instantaneous(Spawn Count=1, Use Spawn Probability=false)
  ]
  particleSpawnScript [&InitializeParticle(Color=<rgba 0.1769 0.0064 0.3073 1>, Lifetime=1.3, Sprite Size Mode=::NewEnumerator3, UsePositionOffset=false, Write Material Random=true)]
  particleUpdateScript [
   &ParticleState()
   &ScaleSpriteSize(Scale Factor=&Vector2DFromFloat(Value=&FloatFromCurve(FloatCurve=DataInterface() {propertyValues {Curve {keys [
          {arriveTangent=0.4346, interpMode=RCIM_Cubic, leaveTangent=0.4346, tangentMode=RCTM_User, time=0, value=0}
          {arriveTangent=0.0473, interpMode=RCIM_Cubic, leaveTangent=0.0473, tangentMode=RCTM_User, time=0.2, value=1}
         ]}}})))
   &ScaleColor(
    Scale Alpha=&FloatFromCurve(
     FloatCurve=DataInterface() {propertyValues {Curve {keys [
         {arriveTangent=0.4005, interpMode=RCIM_Cubic, leaveTangent=0.4005, tangentMode=RCTM_User, time=0, value=0}
         {arriveTangent=0.0256, interpMode=RCIM_Cubic, leaveTangent=0.0256, tangentMode=RCTM_User, time=0.3, value=1}
         <at 1 0 RCIM_Cubic>
        ]}}}
     Scale Curve=&FNTimeOfDayScalar(Night Scale=0.8, Morning Scale=0.8, Evening Scale=0.8)
    )
    ScaleA=true
    ScaleRGB=true
   )
   &CameraOffset(Camera Offset Amount=-20)
   /SetVariables(
    Particles.GlowColor2:LinearColor=&MakeLinearColorFromVectorAndFloat(
     "Vector (RGB)"=&MakeVectorFromLinearColorRGB(Color=<rgba 1.8333 0 2 1>)
     "Float (Alpha)"=&Multiply_Float(
      A=0.1
      B=&FloatFromCurve(FloatCurve=DataInterface() {propertyValues {Curve {keys [{arriveTangent=-1.3442, interpMode=RCIM_Cubic, leaveTangent=-1.3442, tangentMode=RCTM_User, time=0, value=1}, <at 0.9 0 RCIM_Cubic>]}}})
     )
    )
    Particles.GlowSize2:Vector2f=&Vector2DFromFloat(Value=&Multiply_Float(
      A=40
      B=&FloatFromCurve(FloatCurve=DataInterface() {propertyValues {Curve {keys [{arriveTangent=0.4346, time=0, value=1}, {arriveTangent=0.0473, leaveTangent=0.0473, time=1, value=1}]}}})
     ))
   ) {isSetParameters=true}
  ]
  rendererClasses [&NiagaraSpriteRendererProperties]
  renderers [{rendererClass=&NiagaraSpriteRendererProperties}]
 }
 Spark {
  simTarget=GPUComputeSim
  emitterUpdateScript [
   &EmitterState(Life Cycle Mode=::NewEnumerator1, Loop Behavior=::NewEnumerator1, Loop Delay=0.1, Loop Duration=1.5, UseLoopDelay=false)
   &SpawnRate(
    SpawnRate=&FloatFromCurve(
     FloatCurve=DataInterface() {propertyValues {Curve {keys [<at 0 1 RCIM_Cubic>,<at 1 0 RCIM_Cubic>]}}}
     CurveIndex=$Emitter.NormalizedLoopAge
     Scale Curve=14
    )
    Use Spawn Probability=false
   )
   &SpawnBurst_Instantaneous(Spawn Count=22, Use Spawn Probability=false) {enabled=false}
  ]
  particleSpawnScript [
   &InitializeParticle(
    Color=&SelectLinearColorFromArray(Array Sampling Mode=::NewEnumerator0, Direct Array Index=?, Color Selection Array=$User.Color Array, Random Seed=0)
    Lifetime Max=1.5
    Lifetime Min=0.5
    Lifetime Mode=::NewEnumerator1
    Sprite Size Max=<3 6>
    Sprite Size Min=<2 4>
    Sprite Size Mode=::NewEnumerator2
    UsePositionOffset=false
    Write Material Random=true
   )
   &SphereLocation(
    MaskSpawnGroup=false
    Override Seed=false
    Sphere Radius=10
    Surface Only=false
    UseNonUniformScale=true
    UseOffset=false
    UseSphereOrientation=true
   )
   &AddVelocityFromPoint(Module.Use Falloff Distance=false, Velocity Strength=&RandomRangeFloat(Minimum=5, Maximum=80))
   &InheritVelocity(
    Inherited Velocity Amount Scale=&VectorFromFloat(Value=1)
    Inherited Velocity Speed Limit=50
    Speed Limit=true
    UseSpeedThreshold=false
   )
  ]
  particleUpdateScript [
   &ParticleState()
   &ScaleColor(
    Scale Alpha=&FNTimeOfDayScalar(Night Scale=0.6, Morning Scale=0.4, Evening Scale=0.4)
    Scale RGB=&VectorFromFloat(Value=&FloatFromCurve(FloatCurve=DataInterface() {propertyValues {Curve {keys [<at 0 10 RCIM_Cubic>,<at 0.9 1 RCIM_Cubic>]}}}))
    ScaleA=true
    ScaleRGB=true
   ) {module=ScaleColor001}
   &GravityForce(Gravity=&RandomRangeVector(Minimum=<0 0 0>, Maximum=<0 0 -100>))
   &CurlNoiseForce(
    Noise Frequency=20
    Noise Strength=&Multiply_Float(
     A=&RandomRangeFloat(Minimum=50, Maximum=600)
     B=&FloatFromCurve(FloatCurve=DataInterface() {propertyValues {Curve {keys [<at 0 0>,<at 0.2 1>]}}})
    )
    Pan Noise=true
   )
   &Drag(Drag=&RandomRangeFloat(Minimum=3, Maximum=6), Use Linear Drag=true, Use Rotational Drag=false)
   &SolveForcesAndVelocity(Clamp Velocity=false, Limit Acceleration=false)
   &ScaleSpriteSize(Scale Factor=&MakeVector2D(
     X=&FloatFromCurve(FloatCurve=DataInterface() {propertyValues {Curve {keys [<at 0 0>,<at 0.1 1>,<at 1 0.5>]}}})
     Y=&FloatFromCurve(FloatCurve=DataInterface() {propertyValues {Curve {keys [<at 0 0.5>,<at 0.1 1>,<at 1 0>]}}})
    ))
   &ScaleSpriteSizeBySpeed(Max Scale Factor=<0.2 5>, Min Scale Factor=<1 0.5>)
  ]
  rendererClasses [&NiagaraSpriteRendererProperties]
  renderers [{rendererClass=&NiagaraSpriteRendererProperties}]
 }
}
summary {
 systemName=NS_Emote_Swatch_Hand_Particles
 emitters [
  {emitterName=Heart_Glow, rendererClasses [&NiagaraSpriteRendererProperties]}
  {emitterName=Final_Heart_Lit001, rendererClasses [&NiagaraSpriteRendererProperties]}
  {emitterName=Spark, simTarget=GPUComputeSim, rendererClasses [&NiagaraSpriteRendererProperties]}
 ]
 userVariables [
  User.bIsFrontEnd {type=NiagaraBool}
  User.bIsFrontEndPreview {type=NiagaraBool}
  User.Color Array {
   type=NiagaraDataInterfaceArrayColor
   description=""
   defaultValue {
    struct=NiagaraExt_VariableValue_DataInterface
    value {
     dataInterface=&self:NiagaraDataInterfaceArrayColor_0
     dataInterfaceClass=&NiagaraDataInterfaceArrayColor
    }
   }
  }
 ]
}
