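# Niagara sprite effect for the Swatch emote's heart energy: a heart that flickers up and flares bright, with a glow and ring burst.
# Final_Heart_Lit001 animates color, alpha and brightness by curve, spiking late in life.
# Glow001 and Ring_Scale_In fire about a second in, driving extra glow, flare and shrinking-ring attributes.
# Time-of-day scalars tune intensity; skeletal mesh placement and user color array modules are disabled.
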NiagaraSystem /BRCosmetics/Animation/Game/MainPlayer/Emotes/Swatch/FX/NS_Emote_Swatch_Heart_Energy
dependencies {
 usedDataInterfaces [&NiagaraDataInterfaceColorCurve, &NiagaraDataInterfaceCurve]
 usedDynamicInputs []
 usedModules [&SystemState]
}
emitters {
 Final_Heart_Lit001 {
  emitterUpdateScript [
   &EmitterState(Life Cycle Mode=::NewEnumerator1, Loop Behavior=::NewEnumerator1, UseLoopDelay=false)
   &SpawnBurst_Instantaneous(Use Spawn Probability=false, UseLoopCountLimit=false)
  ]
  particleSpawnScript [
   &PartitionParticles(Number Of Partitions=5) {enabled=false}
   &InitializeParticle(Lifetime=1.18, Sprite Size Mode=::NewEnumerator3, Uniform Sprite Size=45, UsePositionOffset=false, Write Material Random=true)
   &CameraOffset(Camera Offset Amount=10)
  ]
  particleUpdateScript [
   &ParticleState()
   &ScaleSpriteSize() {enabled=false}
   &DynamicMaterialParameters(
    Index 0 Param 1=2
    Index 0 Param 4=&RandomRangeFloat(Minimum=0, Maximum=1)
    Module.Param0WriteEnabled=true
    Module.Param0WriteEnabled1=true
    Module.Param1WriteEnabled=true
    Module.Param1WriteEnabled1=true
    Module.Param2WriteEnabled=true
    Module.Param2WriteEnabled1=true
    Module.Param3WriteEnabled=true
    Module.Param3WriteEnabled1=true
    Write Parameter Index 1=true
   )
   &SkeletalMeshLocation(Bone Sampling Mode=::NewEnumerator4, MaskSpawnGroup=false, Skeletal Mesh=$User.SkeletalMesh) {enabled=false}
   &Color(Color=&MultiplyLinearColor(
     Linear Color=&ColorFromCurve(ColorCurve=DataInterface() {propertyValues {
        AlphaCurve {keys [<at 0 1 RCIM_Cubic>]}
        BlueCurve {keys [<at 0.2171 1>,<at 0.47 0.5647>,<at 0.8272 1 RCIM_Cubic>]}
        GreenCurve {keys [<at 0.2171 0.0742>,<at 0.47 0.0844>,<at 0.8272 0 RCIM_Cubic>]}
        RedCurve {keys [<at 0.2171 0.0704>,<at 0.47 0.7835>,<at 0.8272 0.2 RCIM_Cubic>]}
       }})
     Scale Factor=&MakeFloat4FromVectorAndFloat(
      "Vector (RGB)"=&VectorFromFloat(Value=250)
      "Float (Alpha)"=&FNTimeOfDayScalar(Night Scale=0.42, Morning Scale=1, Evening Scale=1)
     )
    ))
   &Color(Color=&MultiplyLinearColor(
     Linear Color=&SelectLinearColorFromArray(
      Array Sampling Mode=::NewEnumerator2
      Direct Array Index=?
      Interpolation=&FloatFromCurve(FloatCurve=DataInterface() {propertyValues {Curve {keys [{tangentMode=RCTM_User, time=0, value=1}, {tangentMode=RCTM_User, time=1, value=0}]}}})
      Color Selection Array=$User.Color Array
      Random Seed=?
     )
     Scale Factor=&MakeFloat4FromVectorAndFloat("Vector (RGB)"=&VectorFromFloat(Value=250), "Float (Alpha)"=1)
    )) {enabled=false, module=Color001}
   &ScaleColor(
    Color Value To Scale=$StackContext.Color
    Scale Alpha=&FloatFromCurve(FloatCurve=DataInterface() {propertyValues {Curve {keys [
         {arriveTangent=0.0461, interpMode=RCIM_Cubic, leaveTangent=0.0461, tangentMode=RCTM_User, time=0, value=0}
         {arriveTangent=0.1222, interpMode=RCIM_Cubic, leaveTangent=0.1222, tangentMode=RCTM_User, time=0.1679, value=0}
         {arriveTangent=0.2837, interpMode=RCIM_Cubic, leaveTangent=0.2837, time=0.2478, value=0.1}
         {arriveTangent=0.2543, interpMode=RCIM_Cubic, leaveTangent=0.2543, tangentMode=RCTM_User, time=0.3441, value=0.05}
         {arriveTangent=0.1818, interpMode=RCIM_Cubic, leaveTangent=0.1818, tangentMode=RCTM_User, time=0.4963, value=0.1}
         {arriveTangent=0.1862, interpMode=RCIM_Cubic, leaveTangent=0.1862, time=0.6265, value=0.4}
         {arriveTangent=0.0962, interpMode=RCIM_Cubic, leaveTangent=0.0962, tangentMode=RCTM_User, time=0.7649, value=0.15}
         <at 0.829 0.15>
         <at 0.8494 0>
         {arriveTangent=0.165, interpMode=RCIM_Cubic, leaveTangent=0.165, tangentMode=RCTM_User, time=0.926, value=0.15}
         {arriveTangent=1.709, interpMode=RCIM_Cubic, leaveTangent=1.709, tangentMode=RCTM_User, time=0.9495, value=1}
        ]}}})
    Scale RGB=&VectorFromFloat(Value=&Multiply_Float(
      A=&FloatFromCurve(FloatCurve=DataInterface() {propertyValues {Curve {keys [<at 0.926 1 RCIM_Cubic>, {arriveTangent=54.0067, interpMode=RCIM_Cubic, leaveTangent=54.0067, tangentMode=RCTM_User, time=0.9495, value=5}]}}})
      B=&Multiply_Float(A=1, B=&FN_FrontEndFloatModifier(Gameplay=1, Front End=1))
     ))
    ScaleA=true
    ScaleRGB=true
   )
  ]
  rendererClasses [&NiagaraSpriteRendererProperties]
  renderers [{rendererClass=&NiagaraSpriteRendererProperties}]
 }
 Glow001 {
  emitterUpdateScript [
   &EmitterState(Life Cycle Mode=::NewEnumerator1, Loop Behavior=::NewEnumerator1, UseLoopDelay=true)
   &SpawnBurst_Instantaneous(Spawn Count=1, Spawn Time=1, Use Spawn Probability=false)
  ]
  particleSpawnScript [
   &InitializeParticle(
    Color=<rgba 0.0461 0.0017 0.08 0.4>
    Lifetime=0.287
    Sprite Size Mode=::NewEnumerator3
    Uniform Sprite Size=300
    UsePositionOffset=false
    Write Material Random=true
   )
   &CameraOffset(Camera Offset Amount=-20)
  ]
  particleUpdateScript [
   &ParticleState()
   &ScaleSpriteSize(Scale Factor=&Vector2DFromFloat(Value=&FloatFromCurve(FloatCurve=DataInterface() {propertyValues {Curve {keys [
          {arriveTangent=0.4346, interpMode=RCIM_Cubic, leaveTangent=0.4346, tangentMode=RCTM_User, time=0, value=0}
          {arriveTangent=0.0473, interpMode=RCIM_Cubic, leaveTangent=0.0473, tangentMode=RCTM_User, time=0.5, value=1}
         ]}}})))
   &ScaleColor(
    Scale Alpha=&FloatFromCurve(
     FloatCurve=DataInterface() {propertyValues {Curve {keys [
         {arriveTangent=0.4005, interpMode=RCIM_Cubic, leaveTangent=0.4005, tangentMode=RCTM_User, time=0, value=0}
         {arriveTangent=0.0256, interpMode=RCIM_Cubic, leaveTangent=0.0256, tangentMode=RCTM_User, time=0.6, value=1}
         <at 1 1 RCIM_Cubic>
        ]}}}
     Scale Curve=&FNTimeOfDayScalar(Night Scale=0.6, Morning Scale=0.8, Evening Scale=0.8)
    )
    ScaleA=true
    ScaleRGB=true
   )
   /SetVariables(
    Particles.GlowColor2:LinearColor=&MakeLinearColorFromVectorAndFloat(
     "Vector (RGB)"=&MakeVectorFromLinearColorRGB(Color=<rgba 0.4427 0.6706 5 1>)
     "Float (Alpha)"=&Multiply_Float(
      A=0.2
      B=&FloatFromCurve(FloatCurve=DataInterface() {propertyValues {Curve {keys [
           {arriveTangent=0.4005, interpMode=RCIM_Cubic, leaveTangent=0.4005, tangentMode=RCTM_User, time=0, value=0}
           {arriveTangent=0.0256, interpMode=RCIM_Cubic, leaveTangent=0.0256, tangentMode=RCTM_User, time=0.2, value=1}
           <at 1 0 RCIM_Cubic>
          ]}}})
     )
    )
    Particles.GlowSize2:Vector2f=&Vector2DFromFloat(Value=&Multiply_Float(
      A=120
      B=&FloatFromCurve(
       FloatCurve=DataInterface() {propertyValues {Curve {keys [
           {arriveTangent=0.4346, interpMode=RCIM_Cubic, leaveTangent=0.4346, tangentMode=RCTM_User, time=0, value=0}
           {arriveTangent=0.0473, interpMode=RCIM_Cubic, leaveTangent=0.0473, tangentMode=RCTM_User, time=0.2, value=1}
          ]}}}
       Scale Curve=&FNTimeOfDayScalar(Night Scale=0.6, Morning Scale=0.4, Evening Scale=0.5)
      )
     ))
   ) {isSetParameters=true}
   /SetVariables(
    Particles.GlowColor03:LinearColor=&MultiplyLinearColor(
     Linear Color=&ColorFromCurve(ColorCurve=DataInterface() {propertyValues {
        AlphaCurve {keys [<at 0 1>,<at 0.1 0.7>,<at 1 0>]}
        BlueCurve {keys [<at 0 1.9193>,<at 0.1 1>,<at 1 1>]}
        GreenCurve {keys [<at 0 0.2604>,<at 0.1 0.0312>,<at 1 0.1719>]}
        RedCurve {keys [<at 0 5>,<at 0.1 0.1766>,<at 1 0.0625>]}
       }})
     Scale Factor=&MakeFloat4FromVectorAndFloat(
      "Vector (RGB)"=<1 1 1>
      "Float (Alpha)"=&FNTimeOfDayScalar(Night Scale=0.3, Morning Scale=0.3, Evening Scale=0.5)
     )
    )
    Particles.GlowSize03:Vector2f=&Vector2DFromFloat(Value=&FloatFromCurve(
      FloatCurve=DataInterface() {propertyValues {Curve {keys [{tangentMode=RCTM_User, time=0, value=0}, {tangentMode=RCTM_User, time=0.1, value=1}, <at 1 3>]}}}
      Scale Curve=220
     ))
   ) {isSetParameters=true}
   &SkeletalMeshLocation(Bone Sampling Mode=::NewEnumerator4, MaskSpawnGroup=false, Skeletal Mesh=$User.SkeletalMesh) {enabled=false}
  ]
  rendererClasses [&NiagaraSpriteRendererProperties]
  renderers [{rendererClass=&NiagaraSpriteRendererProperties}]
 }
 Ring_Scale_In {
  emitterUpdateScript [
   &EmitterState(Life Cycle Mode=::NewEnumerator1, Loop Behavior=::NewEnumerator1, UseLoopDelay=false)
   &SpawnBurst_Instantaneous(Spawn Count=2, Spawn Time=1.05, Use Spawn Probability=false)
  ]
  particleSpawnScript [
   &InitializeParticle(Lifetime=0.15, Sprite Size Mode=::NewEnumerator3, Uniform Sprite Size=500, UsePositionOffset=false, Write Material Random=true)
   /SetVariables(Particles.FlareSize:Vector2f=&Vector2DFromFloat(Value=200)) {isSetParameters=true}
   /SetVariables(Particles.CameraOffset001=20) {isSetParameters=true}
   &CameraOffset(Camera Offset Amount=35)
  ]
  particleUpdateScript [
   &ParticleState()
   &ScaleColor(
    Linear Color Curve=DataInterface() {propertyValues {
      AlphaCurve {keys [<at 0 1>]}
      BlueCurve {keys [<at 0 5>,<at 0.1603 1>,<at 0.3948 0.2256>]}
      GreenCurve {keys [<at 0 0.7812>,<at 0.1603 0.1562>,<at 0.3948 0.0167>]}
      RedCurve {keys [<at 0 0.7812>,<at 0.1603 0.1562>,<at 0.3948 0.8>]}
     }}
    Scale Mode=::NewEnumerator2
   ) {module=ScaleColor001}
   &ScaleColor(
    Scale Alpha=&FNTimeOfDayScalar(Night Scale=1, Morning Scale=1, Evening Scale=1)
    Scale RGB=&VectorFromFloat(Value=&FloatFromCurve(FloatCurve=DataInterface() {propertyValues {Curve {keys [<at 0 5>,<at 0.2 1>,<at 1 0>]}}}, Scale Curve=0))
    ScaleA=true
    ScaleRGB=true
   )
   &ScaleSpriteSize(Scale Factor=&Vector2DFromFloat(Value=&FloatFromCurve(FloatCurve=DataInterface() {propertyValues {Curve {keys [<at 0 2>,<at 0.2 1>,<at 1 0>]}}})))
   /SetVariables(Particles.FlareColor=&MultiplyLinearColor(
     Linear Color=<rgba 20 4.4792 5.3203 1>
     Scale Factor=&MakeFloat4FromVectorAndFloat(
      "Vector (RGB)"=<1 1 1>
      "Float (Alpha)"=&FNTimeOfDayScalar(Night Scale=0.7, Morning Scale=0.5, Evening Scale=0.5)
     )
    )) {isSetParameters=true}
   /SetVariables(Particles.Color001:LinearColor=&MultiplyLinearColor(
     Linear Color=$Particles.FlareColor
     Scale Factor=&MakeFloat4FromVectorAndFloat(
      "Vector (RGB)"=&VectorFromFloat(Value=&FloatFromCurve(FloatCurve=DataInterface() {propertyValues {Curve {keys [<at 0 5>,<at 0.2 1>,<at 1 0>]}}}))
      "Float (Alpha)"=&FNTimeOfDayScalar(Night Scale=0.7, Morning Scale=0.5, Evening Scale=0.7)
     )
    )) {isSetParameters=true}
   /SetVariables(Particles.SpriteSize001:Vector2f=&Multiply_Vector2(
     A=$Particles.FlareSize
     B=&Vector2DFromFloat(Value=&FloatFromCurve(FloatCurve=DataInterface() {propertyValues {Curve {keys [<at 0 2>,<at 0.2 1>,<at 1 0>]}}}))
    )) {isSetParameters=true}
   /SetVariables(
    Particles.RingColor:LinearColor=&ScaleLinearColorByCurve(
     InitialValue=&MakeLinearColorFromVectorAndFloat(
      "Vector (RGB)"=<1 1 1>
      "Float (Alpha)"=&Multiply_Float(A=0.2, B=&FNTimeOfDayScalar(Night Scale=0.8, Morning Scale=0.8, Evening Scale=0.8))
     )
     ScaleCurve=DataInterface() {propertyValues {
       AlphaCurve {keys [<at 0 2>,<at 0.2 0.4>,<at 1 0>]}
       BlueCurve {keys [<at 0 10>,<at 0.1944 1>,<at 1 0.3839>]}
       GreenCurve {keys [<at 0 0.6771>,<at 0.1944 0.0677>,<at 1 0.0521>]}
       RedCurve {keys [<at 0 0.6771>,<at 0.1944 0.4095>,<at 1 1>]}
      }}
     CurveIndex=$Particles.NormalizedAge
    )
    Particles.RingSize:Vector2f=&Vector2DFromFloat(Value=&FloatFromCurve(
      FloatCurve=DataInterface() {propertyValues {Curve {keys [{arriveTangent=2.7977, time=0, value=0}, {tangentMode=RCTM_User, time=0.3, value=1.5}, <at 1 3>]}}}
      Scale Curve=250
     ))
   ) {enabled=false, isSetParameters=true}
   /SetVariables(
    Particles.RingColor:LinearColor=&ScaleLinearColorByCurve(
     InitialValue=&MakeLinearColorFromVectorAndFloat(
      "Vector (RGB)"=<1 1 1>
      "Float (Alpha)"=&Multiply_Float(
       A=&Multiply_Float(
        A=0.2
        B=&FloatFromCurve(FloatCurve=DataInterface() {propertyValues {Curve {keys [<at 0.1 0>,<at 0.5 1>]}}})
       )
       B=&FNTimeOfDayScalar(Night Scale=0.8, Morning Scale=0.8, Evening Scale=0.8)
      )
     )
     ScaleCurve=DataInterface() {propertyValues {
       AlphaCurve {keys [<at 0 2>,<at 0.2 0.4>,<at 1 0>]}
       BlueCurve {keys [<at 0 10>,<at 0.1944 1>,<at 1 0.3839>]}
       GreenCurve {keys [<at 0 0.6771>,<at 0.1944 0.0677>,<at 1 0.0521>]}
       RedCurve {keys [<at 0 0.6771>,<at 0.1944 0.4095>,<at 1 1>]}
      }}
     CurveIndex=$Particles.NormalizedAge
    )
    Particles.RingSize:Vector2f=&Vector2DFromFloat(Value=&FloatFromCurve(
      FloatCurve=DataInterface() {propertyValues {Curve {keys [
          {interpMode=RCIM_Cubic, tangentMode=RCTM_User, time=0, value=1}
          {arriveTangent=-0.8958, interpMode=RCIM_Cubic, leaveTangent=-0.8958, tangentMode=RCTM_User, time=1, value=0}
         ]}}}
      Scale Curve=350
     ))
   ) {isSetParameters=true}
   &SkeletalMeshLocation(Bone Sampling Mode=::NewEnumerator4, MaskSpawnGroup=false, Skeletal Mesh=$User.SkeletalMesh) {enabled=false}
  ]
  rendererClasses [&NiagaraSpriteRendererProperties]
  renderers [{rendererClass=&NiagaraSpriteRendererProperties}]
 }
}
summary {
 systemName=NS_Emote_Swatch_Heart_Energy
 emitters [
  {emitterName=Ring_Scale_In, rendererClasses [&NiagaraSpriteRendererProperties]}
  {emitterName=Glow001, rendererClasses [&NiagaraSpriteRendererProperties]}
  {emitterName=Final_Heart_Lit001, rendererClasses [&NiagaraSpriteRendererProperties]}
 ]
 userVariables [
  User.bIsFrontEnd {type=NiagaraBool}
  User.bIsFrontEndPreview {type=NiagaraBool}
  User.Color Array {
   type=NiagaraDataInterfaceArrayColor
   description=""
   defaultValue {
    struct=NiagaraExt_VariableValue_DataInterface
    value {
     dataInterface=&self:NiagaraDataInterfaceArrayColor_0
     dataInterfaceClass=&NiagaraDataInterfaceArrayColor
    }
   }
  }
  User.SkeletalMesh {
   type=NiagaraDataInterfaceSkeletalMesh
   defaultValue {
    struct=NiagaraExt_VariableValue_DataInterface
    value {dataInterface=&self:NiagaraDataInterfaceSkeletalMesh_0}
   }
  }
 ]
}
